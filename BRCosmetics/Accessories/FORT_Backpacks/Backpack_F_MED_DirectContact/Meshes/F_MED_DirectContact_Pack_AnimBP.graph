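# Animation Blueprint for the DirectContact backpack cosmetic that nudges the pack to follow the character's body pose.
# Each update derives Run and Para alphas from owner-mesh clavicle and spine bone deltas against the reference pose.
# Chained additive Modify Bone nodes apply a constant root offset, then shift and rotate dyn_base_joint by the skydiving flag and those two alphas.
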AnimBlueprint /BRCosmetics/Accessories/FORT_Backpacks/Backpack_F_MED_DirectContact/Meshes/F_MED_DirectContact_Pack_AnimBP
target_skeleton=&/F_MED_DirectContact_Pack_Skeleton
graphs [
 AnimGraph {nodes [
   n1 = "Transform (Modify) Bone - Bone: root"!(ComponentPose=Local To Component(LocalPose=Input Pose()), bAlphaBoolEnabled:bool=True, Translation="0,2,0", Rotation="0,0,0", Scale) -> (Pose) {
    alphaInputType=Bool
    rotationMode=BMM_Additive
    translationMode=BMM_Additive
    alphaBoolBlend {blendOption=Linear}
    boneToModify {boneName=root}
    translation {y=2}
   }
   n2 = "Transform (Modify) Bone - Bone: dyn_base_joint"!(ComponentPose=@n1, bAlphaBoolEnabled:bool=Get bIsSkyDiving(), Translation="0,-2,5", Rotation="20,0,0", Scale="1,1,1") -> (Pose) {
    alphaInputType=Bool
    rotationMode=BMM_Additive
    translationMode=BMM_Additive
    alphaBoolBlend {blendOption=Linear}
    boneToModify {boneName=dyn_base_joint}
    rotation {pitch=20}
    translation {y=-2, z=5}
   }
   n3 = "Transform (Modify) Bone - Bone: dyn_base_joint"!(ComponentPose=@n2, Alpha:float=Get Run() -> (Run:float), Translation="5,0,0", Rotation="-15,0,0", Scale="1,1,1") -> (Pose) {
    rotationMode=BMM_Additive
    translationMode=BMM_Additive
    alphaBoolBlend {blendOption=Linear}
    boneToModify {boneName=dyn_base_joint}
    rotation {pitch=-15}
    translation {x=5}
   }
   n4 = "Transform (Modify) Bone - Bone: dyn_base_joint"(ComponentPose=@n3, Alpha=Get Para() -> (Para:float), Translation="0,0,10", Rotation="0,0,10", Scale="1,1,1") {
    rotationMode=BMM_Additive
    translationMode=BMM_Additive
    alphaBoolBlend {blendOption=Linear}
    boneToModify {boneName=dyn_base_joint}
    rotation {roll=10}
    translation {z=10}
   }
   Output Pose(Result=Component To Local(ComponentPose=@n4)) {layerGroup=None}
  ]}
 EventGraph {nodes [
   Event BlueprintUpdateAnimation()
   n1 = GetOwnerSkeletalMeshComponent()
   > Set Run(Run:float=GetAlphaFromReferenceTransformDelta(SkeletalMeshComponent=@n1, BoneName=clavicle_r, BaseName=spine_05, InputRange=(X=-5,Y=-25), TwistAxis="1,0,0")) -> (Output_Get:float)
   > Set Para(Para:float=GetAlphaFromReferenceTransformDelta(SkeletalMeshComponent=@n1, BoneName=spine_04, BaseName=spine_03, InputRange=(X=0,Y=-15), TwistAxis="1,0,0")) -> (Output_Get:float)
   TryGetPawnOwner()
  ]}
]
variables [
 Para {type=number}
 Run {type=number}
]
